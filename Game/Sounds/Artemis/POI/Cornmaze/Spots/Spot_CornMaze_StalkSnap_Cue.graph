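# Sound cue for a one-shot corn stalk snap spot ambience in the Cornmaze point of interest.
# A random node picks among seven equally weighted stalk-snap wave variations, and every branch of the quality level node routes to that same random node.
# It plays through the ambient one-shots sound class with a wind one-shot tile-ambient attenuation asset, with bus sends disabled.
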SoundCue /Game/Sounds/Artemis/POI/Cornmaze/Spots/Spot_CornMaze_StalkSnap_Cue
root=QualityLevel_0
nodes [
 {id=QualityLevel_0, children [Random_1,Random_1,Random_1,Random_1]}
 {
  id=Random_1
  children [WavePlayer_3, WavePlayer_4, WavePlayer_5, WavePlayer_6, WavePlayer_7, WavePlayer_10, WavePlayer_11]
  properties {weights [1,1,1,1,1,1,1]}
 }
 {id=WavePlayer_10, properties {soundWaveAssetPtr=&/Spot_CornMaze_StalkSnap_06}}
 {id=WavePlayer_11, properties {soundWaveAssetPtr=&/Spot_CornMaze_StalkSnap_07}}
 {id=WavePlayer_3, properties {soundWaveAssetPtr=&/Spot_CornMaze_StalkSnap_01}}
 {id=WavePlayer_4, properties {soundWaveAssetPtr=&/Spot_CornMaze_StalkSnap_02}}
 {id=WavePlayer_5, properties {soundWaveAssetPtr=&/Spot_CornMaze_StalkSnap_03}}
 {id=WavePlayer_6, properties {soundWaveAssetPtr=&/Spot_CornMaze_StalkSnap_04}}
 {id=WavePlayer_7, properties {soundWaveAssetPtr=&/Spot_CornMaze_StalkSnap_05}}
]
properties {
 attenuationSettings=/Game/Sounds/Attenuation/Fort_TileAmbient_OneShot_Wind_Less_Spread_Att
 bEnableBusSends=false
 maxDistance=8500
 soundClassObject=&Fort_Ambient_OneShots
 volumeMultiplier=1.002
 preEffectBusSends [{}]
}
